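# Player animation Blueprint for the Stalker Knife that blends an in-air attack landing onto the lower body.
# Each update reads two animation curves into landing position and blend alpha.
# In MovementMode_FinalPoseOverride, a state machine switches from passthrough to a time-evaluated land sequence when alpha, 2D acceleration and ground checks pass.
# A mesh-space LowerBody blend profile layers that over the cached final pose; other layer graphs contain no logic.
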AnimBlueprint /FNMStalkerKnife/Assets/ANM/AnimBP/StalkerKnife_Player_AnimBP
target_skeleton=&Fortnite_M_Avg_Player_Skeleton
graphs [
 AnimGraph {nodes [Output Pose()]}
 EventGraph {nodes [
   Event BlueprintUpdateAnimation()
   > Set LandPosition(LandPosition:float=GetCurveValue(CurveName=Stalker_LandAnimPosition)) -> (Output_Get:float)
   > Set LandPoseAlpha(LandPoseAlpha:float=GetCurveValue(CurveName=Stalker_LandLowerbodyBlend)) -> (Output_Get:float)
  ]}
 ItemAimOffset {nodes [
   Input Pose!() -> (Pose, AimOffsetAlpha:double, Yaw:double, Pitch:double) {name=InPose_AimOffset}
   Input Pose!() -> (Pose) {name=InPose_PreAimOffsets}
   Output Pose() {layerGroup=Item Override}
  ]}
 ItemFallAdditive {nodes [
   Input Pose() {name=InPoseItemFallAdditive}
   Output Pose() {layerGroup=Item Override}
  ]}
 ItemFallLandAdditive {nodes [
   Input Pose() {name=InPoseItemFallLandAdditive}
   Output Pose() {layerGroup=Item Override}
  ]}
 ItemFallLevel2Override {nodes [
   Input Pose() {name=InPose_FallLevel2}
   Output Pose() {layerGroup=Item Override}
  ]}
 ItemFinalPoseOverride {nodes [
   Input Pose() {name=InFinalPose}
   Output Pose() {layerGroup=Item Override}
  ]}
 ItemFullBodyAimOffsets {nodes [
   Input Pose!() -> (Pose, FullBody_AimOffsetAlpha:double, Aim_Yaw:double, Aim_Pitch:double) {name=InPoseFullBodyAimOffsets}
   Input Pose!() -> (Pose) {name=InPosePreFullBodyAimOffsets}
   Output Pose() {layerGroup=Item Override}
  ]}
 ItemFullBodyOverride {nodes [
   Input Pose() {name=InPoseFullBody}
   Output Pose() {layerGroup=Item Override}
  ]}
 ItemIdleOverride {nodes [
   Input Pose() {name=InPose_Idle}
   Output Pose() {layerGroup=Item Override}
  ]}
 ItemJetPackJumpAdditive {nodes [Output Pose() {layerGroup=Item Override}]}
 ItemJetPackStartAdditive {nodes [Output Pose() {layerGroup=Item Override}]}
 ItemJumpLoopAdditive {nodes [
   Input Pose() {name=InPoseItemJumpLoopAdditive}
   Output Pose() {layerGroup=Item Override}
  ]}
 ItemJumpOverride {nodes [
   Input Pose() {name=InPoseJump}
   Output Pose() {layerGroup=Item Override}
  ]}
 ItemJumpUpAdditive {nodes [
   Input Pose() {name=InPoseItemJumpUpAdditive}
   Output Pose() {layerGroup=Item Override}
  ]}
 ItemPreAimOffsets {nodes [
   Input Pose() {name=InPosePreAimOffsets}
   Output Pose() {layerGroup=Item Override}
  ]}
 ItemPreIK {nodes [
   Input Pose() {name=InPose_PreIK}
   Output Pose() {layerGroup=Item Override}
  ]}
 ItemRightHandIK {nodes [
   Input Pose() {name=InPosePreRightHandIK}
   Input Pose() {name=InPosePostRightHandIK}
   Output Pose() {layerGroup=Item Override}
  ]}
 ItemSkeletalControl {nodes [
   Input Pose()
   Output Pose() {layerGroup=Item Override}
  ]}
 ItemSwimJumpFallLoop {nodes [Output Pose() {layerGroup=Item Override}]}
 ItemSwimJumpSurfaceEnd {nodes [Output Pose() {layerGroup=Item Override}]}
 ItemSwimJumpSurfaceLoop {nodes [Output Pose() {layerGroup=Item Override}]}
 ItemUpperBody {nodes [
   Input Pose!() -> (Pose, InputParam:Fort Anim Input Adjusted Aim Structure) {name=InPoseUpperBody}
   Output Pose() {layerGroup=Item Override}
  ]}
 ItemZipLineStartAdditive {nodes [Output Pose() {layerGroup=Item Override}]}
 MovementMode_AimOffset {nodes [
   Input Pose!() -> (Pose) {name=InAimOffsetsPose}
   Input Pose() {name=InSourcePosePreAimOffsets}
   Output Pose() {layerGroup=MovementMode}
  ]}
 MovementMode_BaseLocomotion {nodes [
   Input Pose() {name=InPoseBaseLocomotion}
   Output Pose() {layerGroup=MovementMode}
  ]}
 MovementMode_FinalPoseOverride {nodes [
   AnimGraphNode_StateMachine_0 = New State Machine() {is=title AnimGraphNode_StateMachine}
   n1 = Profile Blend!(
    BasePose:Pose Link Structure=Use cached pose 'SavedPose32218'()
    BlendPose:Pose Link Structure=@AnimGraphNode_StateMachine_0
    BlendWeight:float=Get LandPoseAlpha() -> (LandPoseAlpha:float)
   ) -> (Pose:Pose Link Structure) {
    is=title AnimGraphNode_BlendProfileLayeredBlend
    blendProfileAsset=/Game/Characters/Player/BlendMasks/LowerBody_BlendMask
    rotationBlendSpace=MeshSpace
    blendPose {linkId=16842751}
   }
   Output Pose(Result=@n1) {layerGroup=MovementMode}
   Save cached pose 'SavedPose32218'(Pose=Input Pose() {name=InFinalPose}) {cacheName=SavedPose32218}
  ]}
 New State Machine {
  entry=Passthrough
  path=MovementMode_FinalPoseOverride.AnimGraphNode_StateMachine_0.New State Machine
  ids=id
  nodes [
   AnimStateNode_0 = Passthrough() {is=title AnimStateNode}
   AnimStateNode_1 = Land() {is=title AnimStateNode}
   AnimStateTransitionNode_0 = AnimStateTransitionNode() {crossfadeDuration=0.01}
   AnimStateTransitionNode_1 = AnimStateTransitionNode() {crossfadeDuration=0}
  ]
  links [
   {
    dst=Passthrough
    rule=MovementMode_FinalPoseOverride.AnimGraphNode_StateMachine_0.New State Machine.AnimStateTransitionNode_1.Transition
    src=Land
    via=AnimStateTransitionNode_1
   }
   {
    dst=Land
    rule=MovementMode_FinalPoseOverride.AnimGraphNode_StateMachine_0.New State Machine.AnimStateTransitionNode_0.Transition
    src=Passthrough
    via=AnimStateTransitionNode_0
   }
  ]
 }
 Passthrough {
  path=MovementMode_FinalPoseOverride.AnimGraphNode_StateMachine_0.New State Machine.AnimStateNode_0.Passthrough
  nodes [Output Animation Pose(Result=Use cached pose 'SavedPose32218'())]
 }
 Land {
  path=MovementMode_FinalPoseOverride.AnimGraphNode_StateMachine_0.New State Machine.AnimStateNode_1.Land
  nodes [Output Animation Pose(Result="Sequence Evaluator (by time) 'Stalker_Inair_Attack_Land'"!(ExplicitTime:float=Get LandPosition() -> (LandPosition:float)) -> (Pose:Pose Link Structure) {is=title AnimGraphNode_SequenceEvaluator, sequence=/FNMStalkerKnife/Assets/ANM/InAirAttack/Stalker_Inair_Attack_Land})]
 }
 Transition {
  path=MovementMode_FinalPoseOverride.AnimGraphNode_StateMachine_0.New State Machine.AnimStateTransitionNode_0.Transition
  nodes [
   n1 = Get LandPoseAlpha()
   Result(bCanEnterTransition=AND Boolean!(A="float > float"(A=@n1), B=Get bIsAccelerating2D(), C:bool=Get bIsOnGround() -> (bIsOnGround:bool)) -> (ReturnValue))
  ]
 }
 Transition {
  path=MovementMode_FinalPoseOverride.AnimGraphNode_StateMachine_0.New State Machine.AnimStateTransitionNode_1.Transition
  nodes [Result(bCanEnterTransition="Equal (Float)"(A=Get LandPosition()))]
 }
 MovementMode_FullBodyOverride {nodes [
   Input Pose() {name=InputPoseFullBody}
   Output Pose() {layerGroup=MovementMode}
  ]}
 MovementMode_InAirOverride {nodes [
   Input Pose() {name=InSource_Pose}
   Input Pose() {name=InAirPose}
   Output Pose() {layerGroup=MovementMode}
  ]}
 MovementMode_LocomotionAdditives {nodes [
   Input Pose() {name=InLocomotionAdditivesPose}
   Output Pose() {layerGroup=MovementMode}
  ]}
 MovementMode_LowerBodyOverride {nodes [
   Input Pose() {name=InLowerBodyPose}
   Input Pose() {name=InSourcePose}
   Output Pose() {layerGroup=MovementMode}
  ]}
 MovementMode_OffScreen {nodes [
   Input Pose() {name=InPoseOffscreen}
   Output Pose() {layerGroup=MovementMode}
  ]}
 MovementMode_SourcePose {nodes [
   Input Pose() {name=InSourcePoseBase}
   Output Pose() {layerGroup=MovementMode}
  ]}
]
variables [
 LandPoseAlpha {type=number}
 LandPosition {type=number}
 Rotation_attach_l {type=Rotator, default=<rot -12 -82 20>}
 Rotation_attach_r {type=Rotator, default=<rot 180 270 0>}
 Translation_attach_l {type=Vector}
 Translation_attach_r {type=Vector, default=<0 0 10>}
]
